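# Sound cue for impacts on a small flower pot, played as a demolition hit sound in the world.
# A mixer layers two branches: a random flower sample at full level and a random ceramic impact at half level.
# Each branch runs through its own modulator that raises pitch by a random amount for variety and a smaller-object feel.
# Small-object demolition attenuation, the impacts sound class and an object-impact concurrency set shape playback.
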SoundCue /Game/Sounds/Fort_Impact_Sounds/Demolition/Flowers/Demo_FlowersPotSmall_Impact_Cue
root=Mixer_0
nodes [
 {id=Mixer_0, children [Modulator_0,Modulator_2], properties {inputVolume [1,0.5]}}
 {id=Modulator_0, children [Random_0], properties {pitchMax=1.5, pitchMin=1.4}}
 {id=Modulator_2, children [Random_1], properties {pitchMax=1.5, pitchMin=1.3}}
 {id=Random_0, children [WavePlayer_0,WavePlayer_1,WavePlayer_2,WavePlayer_3], properties {weights [1,1,1,1]}}
 {id=Random_1, children [WavePlayer_4,WavePlayer_5,WavePlayer_6,WavePlayer_7], properties {weights [1,1,1,1]}}
 {id=WavePlayer_0, properties {soundWaveAssetPtr=&/Demo_Flowers_01}}
 {id=WavePlayer_1, properties {soundWaveAssetPtr=&/Demo_Flowers_02}}
 {id=WavePlayer_2, properties {soundWaveAssetPtr=&/Demo_Flowers_03}}
 {id=WavePlayer_3, properties {soundWaveAssetPtr=&/Demo_Flowers_04}}
 {id=WavePlayer_4, properties {soundWaveAssetPtr=/Game/Sounds/Fort_Impact_Sounds/Demolition/Ceramic/Ceramic_Impact01}}
 {id=WavePlayer_5, properties {soundWaveAssetPtr=/Game/Sounds/Fort_Impact_Sounds/Demolition/Ceramic/Ceramic_Impact02}}
 {id=WavePlayer_6, properties {soundWaveAssetPtr=/Game/Sounds/Fort_Impact_Sounds/Demolition/Ceramic/Ceramic_Impact03}}
 {id=WavePlayer_7, properties {soundWaveAssetPtr=/Game/Sounds/Fort_Impact_Sounds/Demolition/Ceramic/Ceramic_Impact04}}
]
properties {
 attenuationSettings=/Game/Sounds/Attenuation/Fort_Demolition_Hit_Small_Sized_Object_Att
 maxDistance=4000
 soundClassObject=&Fort_Impacts
 volumeMultiplier=0.544
 concurrencyOverrides {maxCount=4, resolutionRule=StopOldest}
 concurrencySet [&CONC_Object_Impacts_FourMax]
}
