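# Sound cue for demolition impact hits on flowers.
# A random node picks one of four equally weighted flower impact waves and feeds it through a modulator for per-play variation.
# It uses small-object demolition hit attenuation, plays in the impacts sound class at reduced volume, and references an object-impact concurrency set that caps simultaneous voices.
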SoundCue /Game/Sounds/Fort_Impact_Sounds/Demolition/Flowers/Demo_Flowers_Impact_Cue
root=Modulator_0
nodes [
 {id=Modulator_0, children [Random_0]}
 {id=Random_0, children [WavePlayer_0,WavePlayer_1,WavePlayer_2,WavePlayer_3], properties {weights [1,1,1,1]}}
 {id=WavePlayer_0, properties {soundWaveAssetPtr=&/Demo_Flowers_01}}
 {id=WavePlayer_1, properties {soundWaveAssetPtr=&/Demo_Flowers_02}}
 {id=WavePlayer_2, properties {soundWaveAssetPtr=&/Demo_Flowers_03}}
 {id=WavePlayer_3, properties {soundWaveAssetPtr=&/Demo_Flowers_04}}
]
properties {
 attenuationSettings=/Game/Sounds/Attenuation/Fort_Demolition_Hit_Small_Sized_Object_Att
 maxDistance=4000
 soundClassObject=&Fort_Impacts
 volumeMultiplier=0.544
 concurrencyOverrides {maxCount=4, resolutionRule=StopOldest}
 concurrencySet [&CONC_Object_Impacts_FourMax]
}
